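# Material function for the anime character materials that outputs UVs from a camera-facing virtual plane anchored to the actor.
# The plane center is actor position plus a parameterized offset, pushed along the camera direction vector.
# The plane normal is that vector biased toward its normalized horizontal component, reducing vertical tilt.
# Camera FOV drives a clamped interpolation of the UV scale fed into VirtualPlaneCoordinates.
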MaterialFunction /Game/Global/Materials/Character_Parent_2023/Anime_Parent_2023/MaterialFunctions/MFC_ActorAlighnedScreenPosition
graph {
 properties {bExposeToLibrary=true, libraryCategoriesText [Misc,TechArt]}
 nodes [
  n1 = MaterialFunctionCall() {materialFunction=&CameraDirectionVector}
  n2 = Multiply(A=Constant3Vector() {constant=<rgba 1 1 0 0>}, B=@n1.Result)
  n3 = Add(A=Multiply(B=Normalize(VectorInput=@n2)) {constA=0.25}, B=@n1.Result)
  n4 = Multiply(A=@n1.Result) {constB=50}
  n5 = VectorParameter() {group=ActorAlighnedScreenPosition, parameterName=ActorAlighnedScreenPosition_ActorPositionOffset, defaultValue=<rgba 0 0 90 1>}
  n6 = Add(A=ActorPositionWS(), B=@n5.RGB)
  n7 = Constant() {r=64}
  n8 = Constant() {r=128}
  n9 = InverseLinearInterpolate(
   A=Constant() {r=45}
   B=Constant() {r=90}
   Value=MaterialFunctionCall() {materialFunction=/Engine/Functions/Engine_MaterialFunctions02/Utility/FOV}.FOV angle in degrees
  ) {bClampResult=true}
  FunctionOutput(None=MaterialFunctionCall(Plane Normal Axis=Normalize(VectorInput=@n3), Plane Center=Add(A=@n4, B=@n6), UV Scale=LinearInterpolate(A=@n7, B=@n8, Alpha=@n9)) {materialFunction=&VirtualPlaneCoordinates}.UVs) {outputName=UVs}
 ]
}
