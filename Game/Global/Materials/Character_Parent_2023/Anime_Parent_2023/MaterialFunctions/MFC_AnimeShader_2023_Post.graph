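# Post stage of the 2023 character anime shader, turning gradient, mask and color inputs into final material attributes.
# Composites lit/shaded color, metallic shading, DFL, specular and optional rim light, then applies LUT, emissive, time-of-day and cinematic color adjustments.
# Outputs Cloth shading model attributes with black BaseColor, color driven through Emissive and data packed into SubsurfaceColor.
# An optional static switch blends toward PBR attributes via PBR_Shading before MF_CharacterParent_2023_PostFX runs.
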MaterialFunction /Game/Global/Materials/Character_Parent_2023/Anime_Parent_2023/MaterialFunctions/MFC_AnimeShader_2023_Post
graph {nodes [
  Composite() {subgraphName=Construct Base Inputs}
  FunctionInput() {inputName=---------, inputType=FunctionInput_Bool, sortPriority=30}
  FunctionInput() {inputName=--------, inputType=FunctionInput_Bool, sortPriority=20}
  n1 = ChannelMaskParameter(None=NamedRerouteUsage()) {group=03_Anime_PBR_Override, maskChannel=Alpha, parameterName=FXChannel_PBR_GMap, sortPriority=10, defaultValue=<rgba 0 0 0 1>}
  n2 = LinearInterpolate(B=OneMinus(None=@n1.R), Alpha=ScalarParameter() {group=03_Anime_PBR_Override, parameterName=PBR_GMap_Intensity, sliderMax=1}) {constA=1}
  n3 = MaterialFunctionCall!(In=NamedRerouteUsage(), LUT, LUT Multiplier) {materialFunction=/Game/Packages/Fortress_SharedMaterials/Anime_Master_Material/MF_ApplyLUT}
  n4 = NamedRerouteUsage()
  n5 = Constant()
  n6 = SetMaterialAttributes!(
   MaterialAttributes
   BaseColor=LinearInterpolate(A=Multiply(A=@n2, B=@n3.Result), B=NamedRerouteUsage(), Alpha=@n4)
   Specular=Multiply(A=NamedRerouteUsage(), B=OneMinus(None=@n4))
   Roughness=LinearInterpolate(A=NamedRerouteUsage(), Alpha=@n4)
   Metallic=NamedRerouteUsage()
   AmbientOcclusion=Constant() {r=1}
   EmissiveColor=Multiply(A=NamedRerouteUsage(), B=ScalarParameter() {group=03_Anime_PBR_Override, parameterName=PBR_Emissive_Intensity, sliderMax=1})
   Opacity=@n5
   SubsurfaceColor=@n5
   CustomData0=@n5
   Normal=NamedRerouteUsage()
  ) {attributeSetTypes [
    #1
    #2
    #3
    #4
    #5
    #6
    #7
    #8
    #9
    #10
   ]}
  n7 = ScalarParameter() {group=03_Anime_PBR_Override, parameterName=PBR_Shading, sliderMax=1, sortPriority=5}
  n8 = Constant()
  n9 = NamedRerouteUsage()
  n10 = SetMaterialAttributes!(
   MaterialAttributes
   BaseColor=@n8
   Metallic=@n9
   Specular=@n8
   Roughness=Constant() {r=1}
   EmissiveColor=NamedRerouteUsage()
   SubsurfaceColor=MaterialFunctionCall(X=NamedRerouteUsage(), Y=@n9, Z=NamedRerouteUsage()) {materialFunction=&MakeFloat3}.Result
   CustomData0=@n8
   Normal=NamedRerouteUsage()
   AmbientOcclusion=OneMinus(None=NamedRerouteUsage())
   ShadingModel=ShadingModel() {shadingModel=MSM_Cloth}
  ) {attributeSetTypes [
    #1
    #4
    #2
    #3
    #6
    #8
    #9
    #10
    #5
    #11
   ]}
  n11 = BlendMaterialAttributes(A=@n10, B=@n6, Alpha=NamedRerouteDeclaration(None=Saturate(None=@n7)) {name=PBR_Switch, nodeColor=<rgba 0 1 0.8566 1>})
  n12 = Constant()
  MaterialExpressionReroute_4 = Reroute(None=@n12)
  n13 = PreSkinnedPosition()
  n14 = ShaderStageSwitch(PixelShader=VertexInterpolator(VS=@n13).PS, VertexShader=@n13)
  MaterialExpressionReroute_5 = Reroute(None=@n14)
  n15 = MaterialFunctionCall!(Material) {materialFunction=/Game/Packages/Fortress_SharedMaterials/Base_Material_Functions/CharacterBrightnessandScale_Balance}
  MaterialExpressionReroute_6 = Reroute(None=@n15.Distance Blend)
  n16 = Fresnel() {baseReflectFraction=0, exponent=1}
  MaterialExpressionReroute_7 = Reroute(None=@n16)
  FunctionOutput(None=MaterialFunctionCall(
    Material=StaticSwitchParameter(True=@n11, False=@n10) {group=02_Anime Parameters, parameterName=UsePBRSwitch, sortPriority=0}
    M=@MaterialExpressionReroute_4.M
    Pre Skinned Local Position=@MaterialExpressionReroute_5.Pre Skinned Local Position
    Distance Blend=@MaterialExpressionReroute_6.Distance Blend
    Fresnel=@MaterialExpressionReroute_7.Fresnel
   ) {materialFunction=&MF_CharacterParent_2023_PostFX}.Result)
  n17 = Ceil(None=DotProduct(A=Custom() {code="return View.DirectionalLightDirection;", description=DirectionalLightVector, inputs []}, B=PixelNormalWS()))
  NamedRerouteDeclaration(None=FunctionInput(Preview=Saturate(None=@n17)) {inputName=Shading_Gradient, inputType=FunctionInput_Scalar, sortPriority=1}) {name=Shading_Curve + Threshold, nodeColor=<rgba 1 0.1875 0 1>}
  n18 = OneMinus(None=Fresnel() {baseReflectFraction=0})
  n19 = FunctionInput(Preview=Saturate(None=Step(Y=@n18) {constX=0.9})) {inputName=RimLight_Gradient, inputType=FunctionInput_Scalar, sortPriority=11, previewValue=<1 1 1 1>}
  NamedRerouteDeclaration(None=@n19) {name=RimLight_Gradient, nodeColor=<rgba 0 1 0.3837 1>}
  n20 = FunctionInput() {inputName=Color_Shaded, sortPriority=32, previewValue=<0.18 0.18 0.18 1>}
  n21 = FunctionInput() {inputName=Color_Lit, sortPriority=31, previewValue=<0.9 0.9 0.9 1>}
  NamedRerouteDeclaration(None=LinearInterpolate(
    A=NamedRerouteDeclaration(None=@n20) {name=BC_Shaded, nodeColor=<rgba 0 0.326 1 1>}
    B=NamedRerouteDeclaration(None=@n21) {name=BC_Lit, nodeColor=<rgba 0 0.326 1 1>}
    Alpha=NamedRerouteUsage()
   )) {name=ColorOut, nodeColor=<rgba 1 0.8672 0 1>}
  n22 = MaterialFunctionCall(In=NamedRerouteUsage()) {materialFunction=&MF_ApplyLUT}
  n23 = NamedRerouteUsage()
  n24 = ScalarParameter() {
   desc="0.0 = Additive Emissive\r\n1.0 = replacing the Diffuse color with the Emissive color"
   group=02_Anime Parameters
   parameterName=EmissiveAdditiveReplace_Blend
   sliderMax=1
  }
  n25 = LinearInterpolate(A=Add(A=@n22.Result, B=@n23), B=LinearInterpolate(A=@n22.Result, B=@n23, Alpha=Desaturation(None=@n23)), Alpha=@n24)
  NamedRerouteDeclaration(None=MaterialFunctionCall!(
    Emissive In=MaterialFunctionCall!(Color=@n25) {materialFunction=/Game/Packages/Fortress_SharedMaterials/Anime_Master_Material/MF_TODColorAdjustment}.Result
    LightingExtrasColorData=NamedRerouteUsage()
    CINE_LightSources=NamedRerouteUsage()
    ShadingCurve=NamedRerouteUsage()
   ) {materialFunction=/Game/Packages/Fortress_SharedMaterials/Cinematics/Anime/MF_CINE_FinalColorBlend}.Result) {name=ColorBuffer, nodeColor=<rgba 1 0.3072 0 1>}
  n26 = NamedRerouteUsage()
  n27 = MaterialFunctionCall!(Float4=NamedRerouteUsage()) {materialFunction=/Game/Global/MaterialFunctions/MF_BreakComponents}
  n28 = LinearInterpolate(B=NamedRerouteUsage(), Alpha=ScalarParameter() {group=02_Anime Parameters, parameterName=SpecularHighlight_FadeInColorLit, sliderMax=1}) {constA=1}
  n29 = Multiply(A=@n27.A, B=Multiply(A=NamedRerouteUsage(), B=NamedRerouteUsage()))
  n30 = LinearInterpolate(
   A=ScalarParameter() {defaultValue=0.2, group=02_Anime Parameters, parameterName=SpecularHighlight_ShadedIntensity, sliderMax=1}
   B=ScalarParameter() {defaultValue=1, group=02_Anime Parameters, parameterName=SpecularHighlight_LitIntensity}
   Alpha=NamedRerouteUsage()
  )
  n31 = LinearInterpolate(A=@n26, B=Add(A=@n26, B=Multiply(A=@n27.RGB, B=@n28)), Alpha=Saturate(None=Multiply(A=@n29, B=@n30))) {desc=apply spec}
  n32 = NamedRerouteUsage()
  n33 = NamedRerouteUsage()
  n34 = NamedRerouteUsage()
  n35 = StaticSwitchParameter(
   True=Multiply(A=ChannelMaskParameter(None=@n33) {group=02_Anime Parameters, parameterName=FXChannel_RimLight, sortPriority=10}.R, B=@n34)
   False=@n34
  ) {group=02_Anime Parameters, parameterName=UseRimLightMasking, sortPriority=0}
  n36 = ScalarParameter() {
   desc="0 = Blends in the Rimlight with the Rimlight Mask\r\n1= Adds the Rimlight color on top of the surfaace color"
   group=02_Anime Parameters
   parameterName=RimLight_AdditiveBlend
   sliderMax=1
  }
  n37 = LinearInterpolate(A=LinearInterpolate(A=@n31, B=@n32, Alpha=@n35), B=Add(A=Multiply(A=@n32, B=@n35), B=@n31), Alpha=@n36)
  NamedRerouteDeclaration(None=Saturate(None=StaticSwitchParameter(True=@n37, False=@n31) {group=02_Anime Parameters, parameterName=UseRimLight, sortPriority=0})) {name=RawColor, nodeColor=<rgba 1 0 0.7969 1>}
  NamedRerouteDeclaration(None=FunctionInput() {inputName=Thickness, inputType=FunctionInput_Scalar, sortPriority=7, previewValue=<1 1 1 1>}) {name=Thickness, nodeColor=<rgba 0.4766 0 1 1>}
  n38 = FunctionInput() {desc="RGB=Color, A=Opacity", inputName=Specular_Color, inputType=FunctionInput_Vector4, sortPriority=36, previewValue=<1 1 1 1>}
  NamedRerouteDeclaration(None=@n38) {name=SpecularColor, desc="RGB=Color, A=Opacity", nodeColor=<rgba 0 0.1989 1 1>}
  NamedRerouteDeclaration(None=FunctionInput() {inputName=NormalBuffer, sortPriority=12, previewValue=<0.5 0.5 1 1>}) {name=NormalBuffer, nodeColor=<rgba 0.2083 0.2083 1 1>}
  NamedRerouteDeclaration(None=FunctionInput() {inputName=Sheen_Mask, inputType=FunctionInput_Scalar, sortPriority=6}) {name=Sheen_Mask, nodeColor=<rgba 0.401 0.8071 1 1>}
  NamedRerouteDeclaration(None=FunctionInput() {inputName=Emissive, sortPriority=34}) {name=Emissive, nodeColor=<rgba 0 0.326 1 1>}
  NamedRerouteDeclaration(None=FunctionInput() {inputName=Specular_Gradient, inputType=FunctionInput_Scalar, sortPriority=5}) {name=Specular_Curve, nodeColor=<rgba 0 1 0.1562 1>}
  NamedRerouteDeclaration(None=FunctionInput() {inputName=DFL_Color, sortPriority=35}) {name=DFL_Color, nodeColor=<rgba 0 0.326 1 1>}
  n39 = NamedRerouteUsage()
  n40 = LinearInterpolate(
   A=NamedRerouteUsage()
   B=@n39
   Alpha=ScalarParameter() {defaultValue=0.5, group=03_Anime_PBR_Override, parameterName=PBR_BaseColor_Shaded_Blend, sliderMax=1}
  )
  n41 = Add(A=@n40, B=ScalarParameter() {group=03_Anime_PBR_Override, parameterName=PBR_BaseColor_Metallic_Bias, sliderMax=1})
  n42 = NamedRerouteUsage()
  NamedRerouteDeclaration(None=StaticSwitchParameter(True=LinearInterpolate(A=@n40, B=@n41, Alpha=@n42), False=@n39) {group=03_Anime_PBR_Override, parameterName=UseLitShadedColorMix}) {name=PBR_BaseColor, nodeColor=<rgba 0.125 1 0 1>}
  NamedRerouteDeclaration(None=LinearInterpolate(A=NamedRerouteUsage(), B=NamedRerouteUsage(), Alpha=NamedRerouteUsage())) {name=RawColor+DF, nodeColor=<rgba 0 0.0156 1 1>}
  NamedRerouteDeclaration(None=FunctionInput() {inputName=DFL_Mask, inputType=FunctionInput_Scalar, sortPriority=8}) {name=DF_Mask, nodeColor=<rgba 1 0.6562 0 1>}
  n43 = FunctionInput() {inputName=Threshold_Gradient, inputType=FunctionInput_Scalar, sortPriority=2, previewValue=<0.5 0 0 1>}
  NamedRerouteDeclaration(None=@n43) {name=Threshold_Curve, nodeColor=<rgba 0.7973 1 0 1>}
  n44 = MaterialFunctionCall(Float3=FunctionInput() {inputName=SRM, sortPriority=33}) {materialFunction=&BreakOutFloat3Components}
  NamedRerouteDeclaration(None=@n44.R) {name=MapInput_PBR_Specular, nodeColor=<rgba 1 0.0765 0 1>}
  NamedRerouteDeclaration(None=FunctionInput() {inputName=CINE_LightSources, sortPriority=21}) {name=CINE_LightSources, nodeColor=<rgba 1 0.9609 0 1>}
  NamedRerouteDeclaration(None=@n44.G) {name=MapInput_PBR_Roughness, nodeColor=<rgba 0 1 0.0508 1>}
  NamedRerouteDeclaration(None=@n44.B) {name=MapInput_PBR_Metallic, nodeColor=<rgba 0.0365 0 1 1>}
  NamedRerouteDeclaration(None=FunctionInput() {inputName=CINE_LightColor, sortPriority=22}) {name=CINE_LightColor, nodeColor=<rgba 1 0.6803 0 1>}
  NamedRerouteDeclaration(None=FunctionInput() {inputName=FX, inputType=FunctionInput_Vector4, sortPriority=9}) {name=FX, nodeColor=<rgba 1 0 0.4219 1>}
  n45 = NamedRerouteUsage()
  n46 = Multiply(A=@n45, B=NamedRerouteUsage())
  n47 = ChannelMaskParameter(None=NamedRerouteUsage()) {group=02_Anime Parameters, parameterName=FXChannel_Metallic, sortPriority=10}
  NamedRerouteDeclaration(None=StaticSwitchParameter(True=LinearInterpolate(A=@n45, B=@n46, Alpha=@n47.R) {desc=apply metallic}, False=NamedRerouteUsage()) {group=02_Anime Parameters, parameterName=UseMetallicShading, sortPriority=0}) {name=MixedColor, nodeColor=<rgba 1 0.375 0 1>}
  NamedRerouteDeclaration(None=FunctionInput() {inputName=Metallic_Gradient, inputType=FunctionInput_Scalar, sortPriority=10}) {name=Metallic_Gradient, nodeColor=<rgba 0.1107 1 0.7992 1>}
  NamedRerouteDeclaration(None=FunctionInput() {desc="RGB=Color", inputName=RimLight_Color, sortPriority=37, previewValue=<1 1 1 1>}) {name=RimLight_Color, desc="RGB=Color, A=Emissive Multiplier", nodeColor=<rgba 1 0.2256 0 1>}
  NamedRerouteUsage()
  PinBase()
  PinBase!(M=@n12, Pre Skinned Local Position=@n14, Distance Blend=@n15.Distance Blend, Fresnel=@n16) {reroutePins [
    M {expression=&self:MaterialExpressionReroute_4}
    Pre Skinned Local Position {expression=&self:MaterialExpressionReroute_5}
    Distance Blend {expression=&self:MaterialExpressionReroute_6}
    Fresnel {expression=&self:MaterialExpressionReroute_7}
   ]}
  VertexNormalWS()
 ]}
