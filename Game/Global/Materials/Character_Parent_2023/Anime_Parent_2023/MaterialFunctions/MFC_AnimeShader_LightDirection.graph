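# Anime character shader material function that outputs the light direction vector for shading.
# It reads the view's directional light direction and color through Custom HLSL nodes.
# Nested If checks use that direction only when the color is above zero and the direction's self dot product is non-zero.
# Otherwise it falls back to SunlightDirectionalVector from FortniteMaterialParameters; atmosphere light direction nodes sit unconnected.
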MaterialFunction /Game/Global/Materials/Character_Parent_2023/Anime_Parent_2023/MaterialFunctions/MFC_AnimeShader_LightDirection
graph {
 properties {bExposeToLibrary=true, description=Light direction for the Anime shader, libraryCategoriesText [TechArt]}
 nodes [
  AtmosphericLightVector()
  Custom() {code="return View.DirectionalLightDirection;", description=DirectionalLightVector, inputs []}
  n1 = Custom() {code="return View.DirectionalLightDirection;", description=DirectionalLightVector, inputs []}
  n2 = DotProduct(A=@n1, B=@n1)
  n3 = CollectionParameter() {
   collection=&FortniteMaterialParameters
   parameterName=SunlightDirectionalVector
  }
  n4 = ComponentMask(None=@n3) {b=true, g=true, r=true}
  FunctionOutput(None=If(
    A=Custom() {code="return View.DirectionalLightColor;", description=DirectionalLightColor, inputs []}
    "A > B"=If(A=@n2, "A > B"=@n1, "A == B"=@n4, "A < B"=@n4) {desc=check if DirectionalLight exists}
    "A == B"=@n4
    "A < B"=@n4
   ) {desc=check if DirectionalLight color is Black}) {outputName=Direction Vector}
  SkyAtmosphereLightDirection()
 ]
}
