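# Material function in the Tech Art library that outputs a 0-1 distance fade factor compensated for viewport FOV.
# Pixel depth is scaled by the tangent of half the horizontal FOV, so blending stays consistent under zoom.
# The Fade Offset input is subtracted, then the result is divided by Fade Distance and saturated.
# A ShadingPathSwitch with a camera-to-world-position distance on Mobile is present but not connected to the output.
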MaterialFunction /Game/Global/Materials/Character_Parent_2023/Anime_Parent_2023/MaterialFunctions/MFC_DistanceFade_FOV_compensation
graph {
 properties {
  bExposeToLibrary=true
  description=Distance blending function which will maintain the the correct blending factor no matter what the viewport FOV is.
  userExposedCaption=DistanceFade with FOV compensation
  libraryCategoriesText [Tech Art]
 }
 nodes [
  n1 = ViewProperty() {property=MEVP_TanHalfFieldOfView}
  n2 = FunctionInput() {bUsePreviewValueAsDefault=true, inputName=Fade Offset, inputType=FunctionInput_Scalar, previewValue=<50 0 0 1>}
  n3 = Subtract(A=Multiply(A=PixelDepth().R, B=ComponentMask(None=@n1.Property) {desc=Horizontal FOV, r=true}), B=@n2)
  n4 = FunctionInput() {bUsePreviewValueAsDefault=true, inputName=Fade Distance, inputType=FunctionInput_Scalar, sortPriority=10, previewValue=<1000 0 0 1>}
  FunctionOutput(None=Saturate(None=Divide(A=@n3, B=@n4)))
  ShadingPathSwitch(Mobile=Distance(A=WorldPosition().XYZ, B=CameraPositionWS()))
 ]
}
