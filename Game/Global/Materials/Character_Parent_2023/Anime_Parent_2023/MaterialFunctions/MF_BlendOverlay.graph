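# Scalar Overlay blend material function in the anime character parent material set, exposed to the library as Blend_Overlay (Scalar).
# An If node tests Base against mid-grey, choosing a doubled multiply of Base and Blend below it and the inverted screen-style form above it.
# The R and G of texture coordinates, split by BreakOutFloat2Components, feed only the Preview pins of the two scalar inputs.
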MaterialFunction /Game/Global/Materials/Character_Parent_2023/Anime_Parent_2023/MaterialFunctions/MF_BlendOverlay
graph {
 properties {bExposeToLibrary=true, description=scalar overlay function, userExposedCaption="Blend_Overlay (Scalar)"}
 nodes [
  n1 = MaterialFunctionCall(Float2=TextureCoordinate()) {materialFunction=&BreakOutFloat2Components}
  n2 = FunctionInput(Preview=@n1.R) {description=Base texture input, inputName=Base, inputType=FunctionInput_Scalar}
  n3 = FunctionInput(Preview=@n1.G) {description=Blend texture input, inputName=Blend, inputType=FunctionInput_Scalar}
  n4 = Multiply(A=Multiply(A=OneMinus(None=@n2), B=Constant() {r=2}), B=OneMinus(None=@n3))
  n5 = Multiply(A=@n2, B=Constant() {r=2})
  FunctionOutput(None=If(A=@n2, B=Constant() {r=0.5}, "A > B"=OneMinus(None=@n4), "A < B"=Multiply(A=@n5, B=@n3)))
 ]
}
