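# Material function that quantizes an NdL input into cel-shading bands with smoothed edges, for the anime character parent material.
# NdL, previewed as half-Lambert from the directional light and vertex normal, is scaled by the floored Bands Amount.
# A SmoothStep around each fractional midpoint is added to the band index and divided by the band count.
# Edge softness combines Width with the DDX/DDY length of the fraction, giving screen-space antialiasing.
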MaterialFunction /Game/Global/Materials/Character_Parent_2023/Anime_Parent_2023/MaterialFunctions/MF_SmoothCelShadingWithBands
graph {nodes [
  n1 = Constant() {r=0.5}
  n2 = Constant()
  n3 = Multiply(A=DotProduct(A=Custom() {code="return View.DirectionalLightDirection;", description=DirectionalLightVector, inputs []}, B=VertexNormalWS())) {constB=0.5}
  n4 = FunctionInput() {inputName=Bands Amount, inputType=FunctionInput_Scalar, sortPriority=20, previewValue=<1 0 0 1>}
  n5 = Floor(None=@n4)
  n6 = Multiply(A=FunctionInput(Preview=Add(A=@n3) {constB=0.5}) {inputName=NdL, inputType=FunctionInput_Scalar}, B=@n5) {constB=3}
  n7 = Frac(None=@n6)
  n8 = AppendVector(A=DDX(Value=@n7), B=DDY(Value=@n7))
  n9 = Multiply(A=Add(
    A=FunctionInput(Preview=@n2) {inputName=Width, inputType=FunctionInput_Scalar, sortPriority=10, previewValue=<0.02 0 0 1>}
    B=Length(None=@n8) {is=class MaterialExpressionLength}
   )) {constB=0.5}
  n10 = SmoothStep(Min=Subtract(A=@n1, B=@n9) {constA=0.5}, Max=Add(A=@n9, B=@n1) {constB=0.5}, Value=@n7)
  n11 = Floor(None=@n6)
  FunctionOutput(None=Divide(A=Add(A=@n10, B=@n11), B=@n5)) {outputName=Divide}
 ]}
